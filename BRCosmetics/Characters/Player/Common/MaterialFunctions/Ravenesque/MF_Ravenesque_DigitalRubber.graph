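# Material function that builds the "digital rubber" surface attributes for the Ravenesque player cosmetic: a non-metallic rubber base color with a subtle glowing accent.
# The mask texture's green channel, replaced by a constant on low/mobile paths, blends roughness between two parameters and also masks emissive.
# Emissive is the accent color scaled by a ColorPanner input and an inverted Fresnel input.
# A blue-channel switch and a time-driven sine are left unconnected.
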MaterialFunction /BRCosmetics/Characters/Player/Common/MaterialFunctions/Ravenesque/MF_Ravenesque_DigitalRubber
graph {nodes [
  n1 = TextureSample() {samplerType=SAMPLERTYPE_LinearColor, texture=/Game/Characters/Player/Common/MaterialFunctions/Ravenesque/Ravenesque_M}
  n2 = MaterialFunctionCall!(Default=@n1.G, Low/Mobile=Constant() {r=0.5}) {materialFunction=/Game/Effects/Fort_Effects/Materials/_Functions/MF_LowMobile_PathingSwitch}
  n3 = TextureSample() {
   samplerType=SAMPLERTYPE_LinearColor
   texture=/Game/Accessories/FORT_Backpacks/Backpack_F_MED_Palespooky_Outlander/Skins/Pancake/Textures/T_Palespooky_Pancake_VertGradient
  }
  n4 = LinearInterpolate(
   B=VectorParameter() {parameterName=RubberAccentColor, defaultValue=<rgba 0.0088 0 0.02 1>}.RGB
   Alpha=FunctionInput(Preview=@n3.R) {inputName=ColorPanner, inputType=FunctionInput_Scalar}
  )
  n5 = Fresnel() {baseReflectFraction=0, exponent=0.25}
  n6 = OneMinus(None=Saturate(None=FunctionInput(Preview=@n5) {inputName=Fresnel}))
  n7 = SetMaterialAttributes!(
   MaterialAttributes
   BaseColor=VectorParameter() {parameterName=RubberColor, defaultValue=<rgba 0.02 0.02 0.02 1>}.RGB
   Roughness=LinearInterpolate(
    A=ScalarParameter() {defaultValue=0.45, parameterName=Roughness_min}
    B=ScalarParameter() {defaultValue=0.55, parameterName=Roughness_max}
    Alpha=@n2.Result
   )
   Specular=Constant() {r=0.4}
   Metallic=Constant()
   EmissiveColor=Multiply(A=@n2.Result, B=Multiply(A=@n4, B=@n6))
  ) {attributeSetTypes [
    #1
    #2
    #3
    #4
    #5
   ]}
  FunctionOutput(None=@n7)
  MaterialFunctionCall(Default=@n1.B) {materialFunction=&MF_LowMobile_PathingSwitch}
  Sine(None=Time()) {period=500}
 ]}
