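# Material function for the Ravenesque player cosmetic that builds a glowing painterly camo surface over a dark, non-metallic base and outputs it as material attributes.
# A mask texture is triplanar-projected with pre-skinned position and normal, keeping the pattern fixed on the animated mesh, with a parameter to fade it.
# Red and green channels blend four colors, while blue plus a Panner gradient input masks them.
# An inverted, powered Fresnel scales the emissive result.
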MaterialFunction /BRCosmetics/Characters/Player/Common/MaterialFunctions/Ravenesque/MF_SL_Ravenesque_PainterlyCamo
graph {nodes [
  n1 = FunctionInput(Preview=VertexInterpolator(VS=PreSkinnedPosition()).PS) {inputName=PreskinnedPosition}
  n2 = PreSkinnedNormal()
  n3 = LinearInterpolate(
   A=MaterialFunctionCall(
    TextureObject=TextureObject() {samplerType=SAMPLERTYPE_LinearColor, texture=/Game/Characters/Player/Common/MaterialFunctions/Ravenesque/Ravenesque_M}
    TextureSize=ScalarParameter() {defaultValue=50, parameterName=TextureSize}
    WorldPosition=@n1
    World Space Normal=VertexInterpolator(VS=@n2).PS
    ProjectionTransitionContrast=Constant() {r=0.5}
   ) {materialFunction=&WorldAlignedTexture}.XYZ Texture
   B=Constant3Vector() {constant=<rgba 0 0.708 0 0>}
   Alpha=ScalarParameter() {parameterName=FadeCamoPattern}
  )
  n4 = FunctionInput(Preview=TextureSample() {samplerType=SAMPLERTYPE_Masks, texture=/Game/Global/Textures/Default/Masks_and_Gradients/T_VertGradient_MASK_2}.G) {inputName=Panner, inputType=FunctionInput_Scalar}
  n5 = Add(A=Subtract(A=ComponentMask(None=@n3) {b=true}), B=Clamp(None=@n4))
  Ceil(None=@n5)
  n6 = Constant()
  n7 = LinearInterpolate(
   A=VectorParameter() {parameterName=ColorA, defaultValue=<rgba 0 0.062 0.1406 1>}.RGB
   B=VectorParameter() {parameterName=ColorB, defaultValue=<rgba 0.0217 0.0833 0.0527 1>}.RGB
   Alpha=@n4
  )
  n8 = ComponentMask(None=@n3) {g=true}
  n9 = VectorParameter() {parameterName=ColorC, defaultValue=<rgba 0.1458 0.0068 0.0426 1>}
  n10 = VectorParameter() {parameterName=ColorD, defaultValue=<rgba 0.0555 0.0152 0.1719 1>}
  n11 = Multiply(
   A=OneMinus(None=@n5)
   B=LinearInterpolate(A=LinearInterpolate(A=@n6, B=@n7, Alpha=@n8), B=LinearInterpolate(A=@n9.RGB, B=@n10.RGB, Alpha=@n4), Alpha=ComponentMask(None=@n3) {r=true})
  )
  n12 = Fresnel() {baseReflectFraction=0, exponent=0.1}
  n13 = OneMinus(None=Power(
    Base=FunctionInput(Preview=@n12) {inputName=Fresnel, inputType=FunctionInput_Scalar}
    Exp=ScalarParameter() {defaultValue=0.06, parameterName=FresnelPower}
   ))
  n14 = SetMaterialAttributes!(
   MaterialAttributes
   BaseColor=VectorParameter() {parameterName=BaseColor, defaultValue=<rgba 0.0092 0.015 0.02 1>}.RGB
   Metallic=Constant()
   Specular=ScalarParameter() {defaultValue=0.2, parameterName=Specular}
   Roughness=Constant() {r=0.66}
   EmissiveColor=Multiply(A=ScalarParameter() {defaultValue=2.5, parameterName=EmissiveBrightness}, B=Multiply(A=@n11, B=@n13))
   Anisotropy=ComponentMask(None=@n3) {b=true, g=true, r=true}
  ) {attributeSetTypes [
    #1
    #2
    #3
    #4
    #5
    #6
   ]}
  FunctionOutput(None=@n14)
  Round(None=@n5)
  Subtract()
  Transform(None=@n1) {transformSourceType=TRANSFORMSOURCE_Local}
 ]}
