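# Material function that outputs a purple, semi-metallic silk material attribute set for the Ravenesque player cosmetic.
# Fixed base color, metallic and anisotropy combine with parameterized roughness and specular.
# A high-exponent Fresnel adds a pinkish emissive rim to fake silk sheen.
# A tiled cloth normal map, replaceable with a flat normal by static switch, passes through FlattenNormal to control weave strength.
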MaterialFunction /BRCosmetics/Characters/Player/Common/MaterialFunctions/Ravenesque/MF_SL_Ravenesque_Silk
graph {nodes [
  n1 = Constant3Vector() {constant=<rgba 0 0 1 0>}
  n2 = TextureSample(UVs=Multiply(A=TextureCoordinate(), B=ScalarParameter() {defaultValue=5, parameterName=TilingCloth})) {samplerType=SAMPLERTYPE_Normal, texture=/BRCosmetics/Characters/Player/Common/MaterialFunctions/Superlevel_DigitalCamo/BasicCloth_N}
  n3 = SetMaterialAttributes!(
   MaterialAttributes
   BaseColor=Constant3Vector() {constant=<rgba 0.135 0.0414 0.1875 0>}
   Metallic=Constant() {r=0.6}
   Roughness=ScalarParameter() {defaultValue=0.42, parameterName=SilkRoughness}
   Anisotropy=Constant3Vector() {constant=<rgba 0 1 0 0>}
   EmissiveColor=LinearInterpolate(B=Constant3Vector() {constant=<rgba 0.1927 0.0492 0.0995 0>}, Alpha=Fresnel() {baseReflectFraction=0, exponent=7}) {constB=0.15}
   Normal=MaterialFunctionCall(
    Normal=StaticSwitchParameter(True=@n1, False=@n2.RGB) {parameterName=RemoveClothNormal}
    Flatness=ScalarParameter() {defaultValue=0.8, parameterName=Flatness}
   ) {materialFunction=&FlattenNormal}.Result
   Specular=ScalarParameter() {defaultValue=0.4, parameterName=Silk Specularity}
  ) {attributeSetTypes [
    #1
    #2
    #3
    #4
    #5
    #6
    #7
   ]}
  FunctionOutput(None=@n3)
 ]}
