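# Material function for Ravenesque player cosmetics that builds a thin-film metal surface switching between gold and gunmetal.
# A scalar Gold/GunMetal input lerps between parameterized gold and darker metal tints and roughness values, with metallic fixed at full.
# These material attributes pass through the shared weapon-skin ThinFilmFunction at full alpha, and its result is the output.
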MaterialFunction /BRCosmetics/Characters/Player/Common/MaterialFunctions/Ravenesque/MF_SL_ThinfilmMetals
graph {nodes [
  n1 = FunctionInput(Preview=Round(None=TextureSample() {texture=/Game/Global/Textures/Default/Masks_and_Gradients/CheckerboardTexture}.G)) {inputName=Gold/GunMetal, inputType=FunctionInput_Scalar}
  n2 = LinearInterpolate(
   A=VectorParameter() {parameterName=Gold Tint, defaultValue=<rgba 1 0.5809 0.1719 1>}.RGB
   B=VectorParameter() {parameterName=Metal Tint, defaultValue=<rgba 0.224 0.2109 0.204 1>}.RGB
   Alpha=@n1
  )
  n3 = Constant() {r=1}
  n4 = LinearInterpolate(
   A=ScalarParameter() {defaultValue=0.2, parameterName=Gold Roughness}
   B=ScalarParameter() {defaultValue=0.33, parameterName=MetalRoughness}
   Alpha=@n1
  )
  FunctionOutput(None=MaterialFunctionCall!(
    Alpha=Constant() {r=1}
    Material Attributes=SetMaterialAttributes!(MaterialAttributes, BaseColor=@n2, Metallic=@n3, Roughness=@n4) {attributeSetTypes [#1, #2, #3]}
   ) {materialFunction=/Game/Weapons/WeaponSkins/MaterialFunction/ThinFilmFunction}.Result)
 ]}
