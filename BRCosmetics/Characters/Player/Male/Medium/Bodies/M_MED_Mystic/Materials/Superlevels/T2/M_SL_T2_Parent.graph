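# Parent character material for Mystic's tier 2 Super Level skins, layering stylized finishes between the shared PreFX and PostFX functions.
# Two mask textures blend Ravenesque silk, painterly camo, thin-film metal, digital rubber and emissive layers over the base.
# A static switch optionally toon-shades everything beneath the emissive layer.
# Base normals are flattened, then recombined after layering; a shared panning gradient swaps to a sine on low/mobile.
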Material /BRCosmetics/Characters/Player/Male/Medium/Bodies/M_MED_Mystic/Materials/Superlevels/T2/M_SL_T2_Parent
graph {
 properties {
  bUseMaterialAttributes=true
  bUsedWithMeshDeformer=true
  bUsedWithMorphTargets=true
  bUsedWithSkeletalMesh=true
  pixelDepthOffsetMode=PDOM_Legacy
  usedShadingModels=Default Lit
 }
 nodes [
  n1 = MaterialFunctionCall() {materialFunction=&BaseFN_CharacterFunction_PreFX}
  n2 = TextureSampleParameter2D() {
   parameterName=Mask01
   samplerSource=SSM_Wrap_WorldGroupSettings
   samplerType=SAMPLERTYPE_Masks
   texture=/BRCosmetics/Characters/Player/Male/Medium/Bodies/M_MED_Mystic/Textures/Skins/Super_Levels/T_Mystic_-_Super_Levels_Body_SL_FX
  }
  n3 = Max(A=Max(A=@n2.R, B=@n2.R), B=@n2.B)
  n4 = TextureSampleParameter2D() {
   parameterName=Mask02
   samplerSource=SSM_Wrap_WorldGroupSettings
   samplerType=SAMPLERTYPE_Masks
   texture=/BRCosmetics/Characters/Player/Male/Medium/Bodies/M_MED_Mystic/Textures/Skins/Super_Levels/T_Mystic_-_Super_Levels_Body_SL_FX_2
  }
  n5 = Max(A=Max(A=@n4.R, B=@n4.R), B=@n4.G)
  n6 = BlendMaterialAttributes(
   A=SetMaterialAttributes!(
    MaterialAttributes=GetMaterialAttributes(None=@n1.Result) {attributeGetTypes [#1]}.MaterialAttributes
    Normal=Constant3Vector() {constant=<rgba 0 0 1 0>}
   ) {attributeSetTypes [#1]}
   B=MaterialFunctionCall() {materialFunction=/BRCosmetics/Characters/Player/Common/MaterialFunctions/Ravenesque/MF_SL_Ravenesque_Silk}.Result
   Alpha=@n2.R
  )
  n7 = MaterialFunctionCall!(Panner=NamedRerouteUsage(), Fresnel=@n1.Fresnel, PreskinnedPosition=@n1.Fresnel) {materialFunction=/BRCosmetics/Characters/Player/Common/MaterialFunctions/Ravenesque/MF_SL_Ravenesque_PainterlyCamo}
  n8 = BlendMaterialAttributes(
   A=BlendMaterialAttributes(
    A=BlendMaterialAttributes(A=@n6, B=@n7.Result, Alpha=@n2.G)
    B=MaterialFunctionCall!(Gold/GunMetal=@n4.R) {materialFunction=/BRCosmetics/Characters/Player/Common/MaterialFunctions/Ravenesque/MF_SL_ThinfilmMetals}.Result
    Alpha=@n2.B
   )
   B=MaterialFunctionCall!(Fresnel=@n1.Fresnel, ColorPanner=NamedRerouteUsage()) {materialFunction=/BRCosmetics/Characters/Player/Common/MaterialFunctions/Ravenesque/MF_Ravenesque_DigitalRubber}.Result
   Alpha=@n4.B
  )
  n9 = MaterialFunctionCall!(
   SkinMask=OneMinus(None=Max(A=@n3, B=@n5))
   In=MaterialFunctionCall!(In=@n8) {materialFunction=/Game/Packages/Fortress_SharedMaterials/Anime_Master_Material/MF_AnimatedVisor}.Result
  ) {materialFunction=/Game/Packages/Fortress_SharedMaterials/Anime_Master_Material/MF_Toonify}
  n10 = GetMaterialAttributes(None=BlendMaterialAttributes(
    A=StaticSwitchParameter(True=@n9.Result, False=@n8) {parameterName=useAnimeShading}
    B=MaterialFunctionCall!(Fresnel=@n1.Fresnel) {materialFunction=/BRCosmetics/Characters/Player/Common/MaterialFunctions/Ravenesque/MF_SL_Ravenesque_Emissive}.Result
    Alpha=Power(Base=@n4.G) {desc=Helps Remove Jaggies}
   )) {attributeGetTypes [#1, #2, #3]}
  n11 = GetMaterialAttributes(None=@n1.Result) {attributeGetTypes [#1]}
  MaterialFunctionCall(
   AO=@n1.AO
   Distance Blend=@n1.AO
   Fresnel=@n1.AO
   Gmap=@n1.AO
   Pre Skinned Local Position=@n1.AO
   Material=SetMaterialAttributes!(
    MaterialAttributes=@n10.MaterialAttributes
    Normal=MaterialFunctionCall(BaseNormal=@n10.Normal, AdditionalNormal=@n11.Normal) {materialFunction=&BlendAngleCorrectedNormals}.Result
    Base Color=@n10.MaterialAttributes
   ) {attributeSetTypes [#1, #2]}
  ) {materialFunction=&BaseFN_CharacterFunction_PostFX}
  n12 = TextureSample(UVs=Panner() {speedX=-0.1, speedY=0.1}) {
   samplerSource=SSM_Wrap_WorldGroupSettings
   samplerType=SAMPLERTYPE_Masks
   texture=/Game/Global/Textures/Default/Masks_and_Gradients/T_VertGradient_MASK_2
  }
  n13 = Time()
  NamedRerouteDeclaration(None=MaterialFunctionCall!(Default=@n12.G, Low/Mobile=Sine(None=@n13) {period=500}) {materialFunction=/Game/Effects/Fort_Effects/Materials/_Functions/MF_LowMobile_PathingSwitch}.Result) {name=Panner Texture, nodeColor=<rgba 0 0.8828 1 1>}
 ]
}
